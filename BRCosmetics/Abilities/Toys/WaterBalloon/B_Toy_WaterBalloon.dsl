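; Throwable water balloon cosmetic toy that splashes on impact and soaks nearby players.
; The server delays collision after spawn, then handles the first hit or overlap, with the overlap path reusing EventHit's stored outputs.
; Splash hides the balloon and plays effects, while a sphere overlap finds player pawns to receive a GameplayEffect.
; Replicating those actors triggers client OnRep message-feed lines naming soaked players, capped by a maximum.
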(event EventBeginPlay
  (bind _spherecollision (Variables|FortPhysicsBall|GetSphereCollision))
  ;[1 1 1] Stop Player From being hit on spawn
  (if (Networking|IsServer)
    (Collision|SetCollisionEnabled _spherecollision "NoCollision")
    (Utilities|FlowControl|Delay (Variables|Throwing|GetThrowCollisionDelay))
    (Collision|SetCollisionEnabled _spherecollision "QueryOnly")
    (if (Variables|Throwing|GetHasFinishedSpawning)
      (Variables|/Script/Engine.RotatingMovementComponent|SetRotationRate "-180.000000,360.000000,200.000000" (Variables|Default|GetRotatingMovement)))))

(event EventSetBallVelocityAndStartTicking
  (params (NewVelocity :type Vector :default "0, 0, 0"))
  ;[1 1 1] Stop Player From being hit on spawn
  (|CallParentFunction NewVelocity)
  (CallFunction|ClientonThrown))

(event Custom|Client%20On%20Thrown
  ;[1 1 1] Stop Player From being hit on spawn
  (Utilities|FlowControl|DoOnce false
    (:Completed
      (Variables|Throwing|SetIs%20Throwing false))))

(event Visuals|EventInitializeToyInstance
  (params (OwningPC :type Fort%20Player%20Controller%20Object%20Reference) (NumTimesSummoned :type Integer))
  ;[1 1 1] Stop Player From being hit on spawn
  (|CallParentFunction OwningPC NumTimesSummoned)
  (Variables|Throwing|SetHasFinishedSpawning true))

(event Collision|EventActorBeginOverlap
  (params (OtherActor :type Actor%20Object%20Reference))
  ;[1 1 1] On hit do overlap check and see how many people got hit
  ;;[1 1 1] Handle pawns with overlap given it causes the player to nudge around if collision is blocking
  (Utilities|Casting|CastToFortPawn OtherActor)
  (if (Math|Boolean|ANDBoolean (Networking|IsServer) (not (Variables|Throwing|GetHit)) (Variables|Throwing|GetHasFinishedSpawning) (not (Variables|Throwing|GetIs%20Throwing)))
    (Utilities|FlowControl|DoOnce false
      (:Completed
        (Variables|FX|SetSplash%20Rotation (Math|Rotator|MakeRotfromZ (stale Collision|EventHit HitNormal)))
        (Variables|FX|SetDecal%20Rotation (Math|Rotator|MakeRotfromX (stale Collision|EventHit HitNormal)))
        (Variables|Throwing|SetHit true)
        (Utilities|IsValid (stale Collision|EventHit Other)
          (:Is%20Valid
            (bind _asfort_pawn (Utilities|Casting|CastToFortPawn (stale Collision|EventHit Other))
              (:then
                (Variables|Throwing|SetPawn _asfort_pawn)
                (Variables|Throwing|SetPawnHit true))
              (:CastFailed
                (Variables|Throwing|SetPawnHit false)))
            (bind _asfort_player_pawn (Utilities|Casting|CastToFortPlayerPawn (stale Collision|EventHit Other))
              (:then
                (Utilities|Casting|CastToFortCharacterVehicle (stale Collision|EventHit Other)
                  (:then
                    _)
                  (:CastFailed
                    (Utilities|Casting|CastToFortPlayerState (Variables|/Script/Engine.Controller|GetPlayerState (Pawn|GetController _asfort_player_pawn)))))
                (Development|PrintString "Balloon Hit")
                (CallFunction|Splash (Variables|FX|GetSplash%20Rotation) (Variables|FX|GetDecal%20Rotation) (Variables|Throwing|GetPawnHit) (Variables|Throwing|GetPawn))
                (bind (_outactors _returnvalue) (Collision|SphereOverlapActors (Transformation|GetActorLocation) (Variables|Targeting|GetSplashZone) (Variables|Targeting|GetSphereCheckObjectTypes) "/Script/FortniteGame.FortPlayerPawn"))
                (if _returnvalue
                  (bind _output_get (Variables|Targeting|SetRepHitActors _outactors))
                  (CallFunction|ApplyGE _output_get)))
              (:CastFailed
                _)))
          (:Is%20Not%20Valid
            _))))))

(event Custom|ApplyGE
  (params (HitActors :type Array%20of%20Actor%20Object%20References))
  ;[1 1 1] Apply GE that shows VFX
  (for _array_element HitActors
    (Environment|ApplyGlobalEnvironmentGameplayEffecttoActor _array_element (Variables|FX|GetGEToApplyFX))))

(event Collision|EventHit
  (params (MyComp :type Primitive%20Component%20Object%20Reference) (Other :type Actor%20Object%20Reference) (OtherComp :type Primitive%20Component%20Object%20Reference) (bSelfMoved :type Boolean) (HitLocation :type Vector :default "0, 0, 0") (HitNormal :type Vector :default "0, 0, 0") (NormalImpulse :type Vector :default "0, 0, 0") (Hit :type Hit%20Result%20Structure%20%28by%20ref%29))
  ;[1 1 1] On hit do overlap check and see how many people got hit
  (|CallParentFunction MyComp Other OtherComp bSelfMoved HitLocation HitNormal NormalImpulse Hit)
  (if (Math|Boolean|ANDBoolean (Networking|IsServer) (not (Variables|Throwing|GetHit)) (Variables|Throwing|GetHasFinishedSpawning) (not (Variables|Throwing|GetIs%20Throwing)))
    (Utilities|FlowControl|DoOnce false
      (:Completed
        (Variables|FX|SetSplash%20Rotation (Math|Rotator|MakeRotfromZ HitNormal))
        (Variables|FX|SetDecal%20Rotation (Math|Rotator|MakeRotfromX HitNormal))
        (Variables|Throwing|SetHit true)
        (Utilities|IsValid Other
          (:Is%20Valid
            (bind _asfort_pawn (Utilities|Casting|CastToFortPawn Other)
              (:then
                (Variables|Throwing|SetPawn _asfort_pawn)
                (Variables|Throwing|SetPawnHit true))
              (:CastFailed
                (Variables|Throwing|SetPawnHit false)))
            (bind _asfort_player_pawn (Utilities|Casting|CastToFortPlayerPawn Other)
              (:then
                (Utilities|Casting|CastToFortCharacterVehicle Other
                  (:then
                    _)
                  (:CastFailed
                    (Utilities|Casting|CastToFortPlayerState (Variables|/Script/Engine.Controller|GetPlayerState (Pawn|GetController _asfort_player_pawn)))))
                (Development|PrintString "Balloon Hit")
                (CallFunction|Splash (Variables|FX|GetSplash%20Rotation) (Variables|FX|GetDecal%20Rotation) (Variables|Throwing|GetPawnHit) (Variables|Throwing|GetPawn))
                (bind (_outactors _returnvalue) (Collision|SphereOverlapActors (Transformation|GetActorLocation) (Variables|Targeting|GetSplashZone) (Variables|Targeting|GetSphereCheckObjectTypes) "/Script/FortniteGame.FortPlayerPawn"))
                (if _returnvalue
                  (bind _output_get (Variables|Targeting|SetRepHitActors _outactors))
                  (CallFunction|ApplyGE _output_get)))
              (:CastFailed
                _)))
          (:Is%20Not%20Valid
            _))))))

(event Custom|Splash
  (params (FXRot :type Rotator) (DecalRot :type Rotator) (HItStatus :type Boolean) (PawnRef :type Fort%20Pawn%20Object%20Reference))
  (bind _returnvalue (Transformation|GetActorLocation))
  ;[1 1 1] When hit, hide everything and Play FX / sounds.
  (Rendering|SetActorHiddenInGame true)
  (Collision|SetActorEnableCollision false)
  (Game|Components|ProjectileMovement|StopSimulating (Variables|Default|GetProjectileMovement) (Variables|Throwing|GetNullHit))
  (Effects|SpawnEmitteratLocation (Variables|FX|GetSplashFX) _returnvalue FXRot)
  (Audio|PlaySoundatLocation (Variables|Sounds|GetSplashSound) _returnvalue)
  (Actor|SetLifeSpan (Variables|Throwing|GetLifespanAfterHit))
  (if HItStatus
    (Effects|SpawnEmitterAttached "/FNGameplayCues/Athena/Items/Consumables/WaterBuffalo/FX/P_WaterBuffalo_Soak.P_WaterBuffalo_Soak" (Variables|/Script/Engine.Character|GetMesh PawnRef) "root")
    (else
      (bind _returnvalue_2 (Game|SpawnActorfromClass "/BRCosmetics/Athena/Items/Consumables/WaterBuffalo/FX/B_Decal_WaterBuffalo.B_Decal_WaterBuffalo_C" (Math|Conversions|ToTransform(Vector) _returnvalue) "Undefined" "MultiplyWithRoot" 0 0))
      (Transformation|SetWorldRotation (Variables|/Script/Engine.Actor|GetRootComponent _returnvalue_2) (Math|Rotator|MakeRotator (+ (.roll DecalRot) (Math|Random|RandomFloatinRange 0.0 3.6e+02)) (.pitch DecalRot) (.yaw DecalRot)) false false))))

(fn OnRep_RepHitActors
  (bind _returnvalue (BlueprintContext|GetFortToyContext))
  (bind _returnvalue_2 (BlueprintContext|GetFortToyContext))
  (bind _returnvalue_3 (BlueprintContext|GetFortToyContext))
  ;[1 1 1] Send messages to clients saying who all got hit.
  (if (not (Networking|IsServer))
    ;;[1 1 1] Check Max people can soak and send messages for each, then send end message and break
    (bind (_array_element _array_index) (Utilities|Array|ForEachLoopwithBreak (Variables|Targeting|GetRepHitActors))
      (:LoopBody
        (Math|Integer|CompareInt _array_index (Variables|Targeting|GetMaxPeopleCanSoak)
          (:>
            _)
          (:==
            _)
          (:<
            (Utilities|IsValid _array_element
              (:Is%20Valid
                (bind _asfort_player_state (Utilities|Casting|CastToFortPlayerState (Actor|GetOwner)))
                (bind _returnvalue_4 (Class|FortToyContext|FormatPlayerNameforDisplay _returnvalue _asfort_player_state "ToyOwner"))
                (bind _asfort_player_pawn (Utilities|Casting|CastToFortPlayerPawn _array_element))
                (bind _asfort_player_state_2 (Utilities|Casting|CastToFortPlayerState (Variables|/Script/Engine.Pawn|GetPlayerState _asfort_player_pawn)))
                (bind _returnvalue_5 (Class|FortToyContext|FormatPlayerNameforDisplay _returnvalue _asfort_player_state_2 "OtherPlayerInteractingWithToy"))
                (Utilities|IsValid _returnvalue_3
                  (:Is%20Valid
                    (Class|FortToyContext|ConditionalPrinttoMessageFeed _returnvalue_3 _asfort_player_state (Transformation|GetActorLocation) "Anyone" (Utilities|Text|FormatText (Variables|Text|GetTextToDisplay) _returnvalue_4 _returnvalue_5))))))))
        ;;[1 1 1] Print Max Hit Message
        (Utilities|IsValid _returnvalue_2
          (:Is%20Valid
            (bind _asfort_player_state_3 (Utilities|Casting|CastToFortPlayerState (Actor|GetOwner)))
            (Class|FortToyContext|ConditionalPrinttoMessageFeed _returnvalue_2 _asfort_player_state_3 (Transformation|GetActorLocation) "Anyone" (Variables|Text|GetMaxHitText))))))))

(fn ConstructionScript
  (|CallParentFunction))
